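# Translucent unlit material for the Proximity Mine disc, a blue glow that stays partly visible behind geometry.
# A sine-driven panning beam texture and a faint grid blend dark and light colors and, with Fresnel, shape opacity.
# With depth testing disabled, a manual scene-versus-pixel depth comparison fades and desaturates occluded areas.
# Vertex red lightens color, alpha masks the main opacity, and blue adds a faint distance-faded outer ring that ignores occlusion.
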Material /Game/Gadgets/Assets/ProximityMine/Material/M_ProximityMine_Disc_01
graph {
 properties {bDisableDepthTest=true, blendMode=BLEND_Translucent, pixelDepthOffsetMode=PDOM_Legacy, shadingModel=MSM_Unlit, usedShadingModels=Unlit}
 nodes [
  n1 = Sine(None=Time()) {period=1.4}
  n2 = TextureSample!(UVs=Panner(Coordinate=TextureCoordinate() {uTiling=12}, Time=Multiply(A=@n1) {constB=12}) {speedX=1}, Tex, Level, Apply View MipBias) {
   constMipValue=0
   mipValueMode=TMVM_MipLevel
   samplerType=SAMPLERTYPE_LinearColor
   texture=/Game/Effects/Fort_Effects/Textures/Tile/T_Raygun_BeamReveal_01
  }
  n3 = TextureCoordinate() {uTiling=2}
  n4 = Add(
   A=@n2.R
   B=Multiply(
    A=TextureSample!(UVs=@n3, Tex, Apply View MipBias) {samplerType=SAMPLERTYPE_LinearColor, texture=/Game/Gadgets/Assets/AirStrike/Textures/T_Airstrike_Grid_01}.R
    B=ScalarParameter() {defaultValue=0.07, parameterName=Grid A Opacity}
   )
  )
  n5 = Multiply(
   A=MaterialFunctionCall(
    Brightness=ScalarParameter() {defaultValue=4, parameterName=Frensel Brightness}
    Exponent=ScalarParameter() {defaultValue=0.8, parameterName=Frensel Exponent}
   ) {materialFunction=&MF_Fresnel_FX}.Fresnel
   B=LinearInterpolate(
    A=ScalarParameter() {defaultValue=0.05, parameterName=Opacity Darks}
    B=ScalarParameter() {defaultValue=1, parameterName=Opacity Lights}
    Alpha=@n4
   )
  )
  n6 = VertexColor()
  n7 = Constant() {r=1}
  n8 = Divide(A=Subtract(A=PixelDepth().R, B=SceneDepth().R), B=ScalarParameter() {defaultValue=300, parameterName=Depth Fade Behind Scale})
  n9 = If(A=SceneDepth().R, B=PixelDepth().R, "A > B"=@n7, "A == B"=@n7, "A < B"=OneMinus(None=Clamp(None=@n8)))
  n10 = VertexColor()
  n11 = ScalarParameter() {defaultValue=0.03, parameterName=Outer Depth Ring Opacity}
  n12 = MaterialFunctionCall(
   Fade_Start_Distance=ScalarParameter() {defaultValue=1000, parameterName=Ring Start Fade, sliderMax=1000}
   Fade_End_Distance=ScalarParameter() {defaultValue=600, parameterName=Ring End Fade, sliderMax=1000}
  ) {materialFunction=&MF_Distance_SyncedFade_FX}
  Add(A=Multiply(A=Multiply(A=@n5, B=@n6.A), B=@n9), B=Multiply(A=Multiply(A=@n10.B, B=@n11), B=OneMinus(None=@n12.Result)))
  n13 = Add(A=@n4, B=VertexColor().R)
  n14 = ScalarParameter() {defaultValue=3, parameterName=Desaturation Exponent}
  Desaturation(
   None=LinearInterpolate(
    A=VectorParameter() {parameterName=Color Dark, defaultValue=<rgba 0 0.31 1 1>}.RGB
    B=VectorParameter() {parameterName=Color Light, defaultValue=<rgba 0.8703 1.328 4 1>}.RGB
    Alpha=Clamp(None=@n13)
   )
   Fraction=OneMinus(None=Power(Base=@n9, Exp=@n14))
  )
 ]
}
