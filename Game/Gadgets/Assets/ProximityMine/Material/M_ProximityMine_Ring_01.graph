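# Translucent unlit mesh-particle material for the proximity mine's ring, tinted red or blue by the TeamColor dynamic parameter.
# An inverted Fresnel blends inner and outer colors and opacity, with particle alpha scaling the opacity.
# Depth testing is disabled, so the ring shows through geometry.
# A pixel-versus-scene depth comparison fades and desaturates occluded parts the farther they sit behind the occluder.
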Material /Game/Gadgets/Assets/ProximityMine/Material/M_ProximityMine_Ring_01
graph {
 properties {
  bDisableDepthTest=true
  bEnableResponsiveAA=true
  bUsedWithMeshParticles=true
  blendMode=BLEND_Translucent
  pixelDepthOffsetMode=PDOM_Legacy
  previewMesh=/Game/Gadgets/Assets/ProximityMine/Mesh/Sm_ProximityMine_Ring_01_NoLOD
  shadingModel=MSM_Unlit
  usedShadingModels=Unlit
 }
 nodes [
  n1 = MaterialFunctionCall(Exponent=ScalarParameter() {defaultValue=4, parameterName=Fresnel Exponent}) {materialFunction=&MF_Fresnel_FX}
  n2 = LinearInterpolate(
   A=VectorParameter() {parameterName=Color Inner, defaultValue=<rgba 0.26 0 0.003 1>}.RGB
   B=VectorParameter() {parameterName=Color Outer, defaultValue=<rgba 2.5 0.3375 0.3375 1>}.RGB
   Alpha=@n1.Fresnel Inverted
  )
  n3 = LinearInterpolate(
   A=VectorParameter() {parameterName=Color Inner, defaultValue=<rgba 0 0.0433 0.26 1>}.RGB
   B=VectorParameter() {parameterName=Color Outer, defaultValue=<rgba 0.3375 1.1665 2.5 1>}.RGB
   Alpha=@n1.Fresnel Inverted
  )
  n4 = DynamicParameter() {defaultValue=<rgba 0 0 0 1>, paramNames [Param1,Param2,Param3,TeamColor]}
  n5 = Constant() {r=1}
  n6 = Divide(A=Subtract(A=PixelDepth().R, B=SceneDepth().R), B=ScalarParameter() {defaultValue=300, parameterName=Depth Fade Behind Scale})
  n7 = If(A=SceneDepth().R, B=PixelDepth().R, "A > B"=@n5, "A == B"=@n5, "A < B"=OneMinus(None=Clamp(None=@n6)))
  n8 = Power(Base=@n7, Exp=ScalarParameter() {defaultValue=3, parameterName=Desaturation Exponent})
  MaterialFunctionCall(Emissive=Desaturation(None=LinearInterpolate(A=@n2, B=@n3, Alpha=@n4.TeamColor) {constAlpha=1}, Fraction=OneMinus(None=@n8))) {materialFunction=&MF_FX_EmissiveAdjustments}
  n9 = ScalarParameter() {defaultValue=0.2, parameterName=Opacity Inner}
  n10 = ScalarParameter() {defaultValue=1, parameterName=Opacity Outer}
  Multiply(A=Multiply(A=LinearInterpolate(A=@n9, B=@n10, Alpha=@n1.Fresnel Inverted), B=ParticleColor().A), B=@n7)
 ]
}
