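# Physics asset for the proximity mine gadget's skeletal mesh, defining its simulation and collision shape.
# It holds a single body on the root bone, a flat box slightly offset from the bone and yawed a quarter turn.
# No constraints are defined, so the mine acts as one rigid piece.
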PhysicsAsset /Game/Gadgets/Assets/ProximityMine/Mesh/SK_ProxMine_Physics
bodies {root [{shapeType=Box, extentX=69.3, extentY=70.8, extentZ=29.8, center=<1 -6 9.6201>, rotation=<rot 0 -90 0>}]}
constraints []
